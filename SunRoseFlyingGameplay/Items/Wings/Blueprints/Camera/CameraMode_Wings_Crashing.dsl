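; Flying camera mode for crashing with wings, adjusting view rotation during a pre-crash windup.
; While time since the last active timestamp is within the windup duration, it zeroes roll delta and leaves yaw untouched.
; Pitch above the initial crash look pitch is eased down toward it by delta time and interp speed, unless input already lowers it faster.
; After the windup, rotation deltas pass through unchanged.
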(fn BP_ProcessViewRotation
  (params (InViewRotation :type Rotator%20%28by%20ref%29 :default "0, 0, 0") (InDeltaRot_Roll :type Float%20%28single-precision%29) (InDeltaRot_Pitch :type Float%20%28single-precision%29) (InDeltaRot_Yaw :type Float%20%28single-precision%29) (DeltaTime :type Float%20%28single-precision%29))
  (if (> (- (Utilities|Time|GetTimeSeconds) (Variables|FortCameraMode_Flying|GetLastActiveTimeSeconds)) (ScalableFloat|GetValueatLevel (Variables|Default|GetRow_PreCrashWindupDuration)))
    (return false InViewRotation InDeltaRot_Roll InDeltaRot_Pitch InDeltaRot_Yaw)
    (else
      (return false InViewRotation 0.0 (Math|Float|SelectFloat (Math|Rotator|NormalizeAxis InDeltaRot_Pitch) (Math|Float|Min(Float) (Utilities|Operators|Multiply (- (ScalableFloat|GetValueatLevel (Variables|Default|GetRow_Crashing_LookPitch_Initial)) (+ (Math|Rotator|NormalizeAxis (.pitch InViewRotation)) (Math|Rotator|NormalizeAxis InDeltaRot_Pitch))) (Variables|GetDeltaTime) (ScalableFloat|GetValueatLevel (Variables|Default|GetRow_Crashing_LookPitch_InterpSpeed))) (Math|Rotator|NormalizeAxis InDeltaRot_Pitch)) (<= (+ (Math|Rotator|NormalizeAxis (.pitch InViewRotation)) (Math|Rotator|NormalizeAxis InDeltaRot_Pitch)) (ScalableFloat|GetValueatLevel (Variables|Default|GetRow_Crashing_LookPitch_Initial)))) InDeltaRot_Yaw))))
