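; Camera modifier for the SunRose Wings flight item that shifts view location and FOV during flight.
; Each frame, unless native calculation is enabled, it reads FlyingPawnComponent status, speeds and input, and timestamps boost, crash and dodge starts.
; Interpolated per-state FOV offsets also feed forward offsets, alongside downward velocity and crash phases.
; Strafe, turn rate and dodges drive sideways offsets; offsets scale by modifier alpha, leaving rotation unchanged.
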(event EventBlueprintModifierEnabled
  (|CallParentFunction)
  (if (not (CallFunction|UseNativeCalculation))
    (Variables|FinalOffsets|SetViewLocationOffset_Right 0.0)
    (Variables|FinalOffsets|SetViewLocationOffset_Forward 0.0)
    (Variables|FinalOffsets|SetViewLocationOffset_Up 0.0)
    (Variables|RuntimeVars|FOV|SetFOVOffset_Mode 0.0)
    (Variables|RuntimeVars|FOV|SetFOVOffset_Speed 0.0)
    (Variables|RuntimeVars|FOV|SetFOVOffset_Input 0.0)
    (Variables|FinalOffsets|SetFOVOffset 0.0)
    (Variables|TimeControls|SetHasSetBoostStartTime false)
    (Variables|TimeControls|SetHasSetCrashStartTime false)
    (Variables|RuntimeVars|SetBaselineFlyingSpeedWithHeightDecay (Math|Float|Sqrt (+ (Math|Float|Square (ScalableFloat|GetValueatLevel (Variables|Balance|Speed|GetRow_FlyingSpeed_Base))) (Math|Float|Square (ScalableFloat|GetValueatLevel (Variables|Balance|GetRow_Flying_HeightDecay_Min))))))
    (Variables|RuntimeVars|FOV|SetClampedPreviousFOVOffset 0.0)
    (Variables|RuntimeVars|FOV|SetInterpolatedClampedDeltaFOV 0.0)
    (Variables|TimeControls|SetTime_BoostStart -1.0)
    (Variables|TimeControls|SetTimeCrashStart -1.0)
    (Variables|TimeControls|SetHasSetDodgeStartTime false)
    (Variables|TimeControls|SetTimeDodgeStart -1.0)
    (bind _asfort_player_controller (Utilities|Casting|CastToFortPlayerController (Game|Player|GetOwningPlayerController (Variables|CameraModifier|GetCameraOwner))))
    (bind _asfort_player_pawn (Utilities|Casting|CastToFortPlayerPawn (Pawn|GetViewTarget _asfort_player_controller)))
    (bind (_returnvalue_roll _returnvalue_pitch _returnvalue_yaw) (Math|Rotator|BreakRotator (Pawn|Components|CharacterMovement|GetLastUpdateRotation (Variables|/Script/Engine.Character|GetCharacterMovement _asfort_player_pawn))))
    (Variables|RuntimeVars|SetPreviousRotation _returnvalue_yaw)))

(fn BlueprintModifyCamera
  (params (DeltaTime :type Float%20%28single-precision%29) (ViewLocation :type Vector :default "0, 0, 0") (ViewRotation :type Rotator :default "0, 0, 0") (FOV :type Float%20%28single-precision%29))
  (bind _baselineflyingspeedwithheightdecay (Variables|RuntimeVars|GetBaselineFlyingSpeedWithHeightDecay))
  (bind _forwardspeed (Variables|RuntimeVars|GetForwardSpeed))
  (bind _returnvalue (>= _forwardspeed _baselineflyingspeedwithheightdecay))
  (bind _baselineflyingspeedwithheightdecay_2 (Variables|RuntimeVars|GetBaselineFlyingSpeedWithHeightDecay))
  (bind _returnvalue_2 (* _baselineflyingspeedwithheightdecay_2 0.7))
  (bind _baselineflyingspeedwithheightdecay_3 (Variables|RuntimeVars|GetBaselineFlyingSpeedWithHeightDecay))
  (bind _forwardspeed_2 (Variables|RuntimeVars|GetForwardSpeed))
  (bind _output (Utilities|RemapFloat _forwardspeed_2 (Math|Vector2D|MakeVector2D _baselineflyingspeedwithheightdecay_3 _returnvalue_2) (0.0 -3e+01) true))
  (bind _maxfovoffset (Variables|TuningControls|FOV|GetMaxFOVOffset))
  (bind _row_flyingspeed_addedatmaxdistancefromceilingheight (Variables|Balance|Speed|GetRow_FlyingSpeed_AddedAtMaxDistanceFromCeilingHeight))
  (bind _returnvalue_3 (ScalableFloat|GetValueatLevel _row_flyingspeed_addedatmaxdistancefromceilingheight))
  (bind _row_flyingspeed_base (Variables|Balance|Speed|GetRow_FlyingSpeed_Base))
  (bind _returnvalue_4 (ScalableFloat|GetValueatLevel _row_flyingspeed_base))
  (bind _returnvalue_5 (+ _returnvalue_4 _returnvalue_3))
  (bind _baselineflyingspeedwithheightdecay_4 (Variables|RuntimeVars|GetBaselineFlyingSpeedWithHeightDecay))
  (bind _forwardspeed_3 (Variables|RuntimeVars|GetForwardSpeed))
  (bind _output_2 (Utilities|RemapFloat _forwardspeed_3 (Math|Vector2D|MakeVector2D _baselineflyingspeedwithheightdecay_4 _returnvalue_5) (Math|Vector2D|MakeVector2D 0.0 _maxfovoffset) true))
  (bind _returnvalue_6 (Math|Float|SelectFloat _output_2 _output _returnvalue))
  (bind (_returnvalue_roll _returnvalue_pitch _returnvalue_yaw) (Math|Rotator|BreakRotator (Pawn|Components|CharacterMovement|GetLastUpdateRotation (Variables|/Script/Engine.Character|GetCharacterMovement (Variables|RuntimeVars|GetPlayerPawn)))))
  (bind (_returnvalue_roll_2 _returnvalue_pitch_2 _returnvalue_yaw_2) (Math|Rotator|BreakRotator (Pawn|Components|CharacterMovement|GetLastUpdateRotation (Variables|/Script/Engine.Character|GetCharacterMovement (Variables|RuntimeVars|GetPlayerPawn)))))
  (bind (_returnvalue_roll_3 _returnvalue_pitch_3 _returnvalue_yaw_3) (Math|Rotator|BreakRotator (Pawn|Components|CharacterMovement|GetLastUpdateRotation (Variables|/Script/Engine.Character|GetCharacterMovement (Variables|RuntimeVars|GetPlayerPawn)))))
  (bind _deltatime (Variables|GetDeltaTime))
  (bind _fovoffset_speed (Variables|RuntimeVars|FOV|GetFOVOffset_Speed))
  (bind _returnvalue_7 (Math|Interpolation|FInterpTo _fovoffset_speed _returnvalue_6 _deltatime 2.75))
  (bind (_asb_athena_sun_rose_wings _bsuccess) (Utilities|Casting|CastToB_Athena_SunRose_Wings (Variables|/Script/FortniteGame.FortPawn|GetCurrentWeapon (Variables|RuntimeVars|GetPlayerPawn))))
  (if (CallFunction|UseNativeCalculation)
    _
    (elif _bsuccess
      (if (> DeltaTime 0)
        (Utilities|FlowControl|Sequence
          (:then_0
            ;[1 1 1] Init Vars
            (Utilities|IsValid (Variables|RuntimeVars|GetPlayerPawn)
              (:Is%20Not%20Valid
                (bind _asfort_player_controller (Utilities|Casting|CastToFortPlayerController (Game|Player|GetOwningPlayerController (Variables|CameraModifier|GetCameraOwner))))
                (Variables|RuntimeVars|SetPlayerController _asfort_player_controller)
                (bind _asfort_player_pawn (Utilities|Casting|CastToFortPlayerPawn (Pawn|GetViewTarget (Variables|RuntimeVars|GetPlayerController))))
                (Variables|RuntimeVars|SetPlayerPawn _asfort_player_pawn)
                (Variables|RuntimeVars|SetLastViewRotation (Variables|GetViewRotation))
                (Variables|Default|SetFlyingPawnComponent (Actor|GetComponentbyClass (Variables|RuntimeVars|GetPlayerPawn) "/SunRoseFlyingGameplay/Items/Wings/Blueprints/FlyingPawnComponent.FlyingPawnComponent_C"))
                (Variables|RuntimeVars|SetPreviousRotation _returnvalue_yaw))))
          (:then_1
            ;[1 1 1] Init Timers
            (Utilities|FlowControl|Sequence
              (:then_0
                (if (Class|FortWingedFlyingBlueprintLibrary|FlyingStatusHasAnyFlags (Variables|/Script/SunRoseFlyingRuntime.FortPawnComponent_WingedFlying|GetFlyingStatus (Variables|Default|GetFlyingPawnComponent)) 1)
                  (if (not (Variables|TimeControls|GetHasSetBoostStartTime))
                    (Variables|TimeControls|SetHasSetBoostStartTime true)
                    (Variables|TimeControls|SetTime_BoostStart (Fortnite|Sync|Time|GetServerWorldTimeSeconds)))
                  (else
                    (Variables|TimeControls|SetHasSetBoostStartTime false))))
              (:then_1
                (if (Class|FortWingedFlyingBlueprintLibrary|FlyingStatusHasAnyFlags (Variables|/Script/SunRoseFlyingRuntime.FortPawnComponent_WingedFlying|GetFlyingStatus (Variables|Default|GetFlyingPawnComponent)) 4)
                  (if (not (Variables|TimeControls|GetHasSetCrashStartTime))
                    (Variables|TimeControls|SetHasSetCrashStartTime true)
                    (Variables|TimeControls|SetTimeCrashStart (Fortnite|Sync|Time|GetServerWorldTimeSeconds)))
                  (else
                    (Variables|TimeControls|SetHasSetCrashStartTime false))))
              (:then_2
                (if (Class|FortWingedFlyingBlueprintLibrary|FlyingStatusHasAnyFlags (Variables|/Script/SunRoseFlyingRuntime.FortPawnComponent_WingedFlying|GetFlyingStatus (Variables|Default|GetFlyingPawnComponent)) 24)
                  (if (not (Variables|TimeControls|GetHasSetDodgeStartTime))
                    (Variables|TimeControls|SetHasSetDodgeStartTime true)
                    (Variables|TimeControls|SetTimeDodgeStart (Fortnite|Sync|Time|GetServerWorldTimeSeconds)))
                  (else
                    (Variables|TimeControls|SetHasSetDodgeStartTime false))))))
          (:then_2
            ;[1 1 1] Set Vars from Pawn Component
            (Variables|RuntimeVars|SetForwardSpeed (Variables|/Script/SunRoseFlyingRuntime.FortPawnComponent_WingedFlying|GetForwardSpeed (Variables|Default|GetFlyingPawnComponent)))
            (Variables|RuntimeVars|SetAddedDownZVelocity (Variables|/Script/SunRoseFlyingRuntime.FortPawnComponent_WingedFlying|GetAddedDownZVelocity (Variables|Default|GetFlyingPawnComponent))))
          (:then_3
            ;[1 1 1] Update elapsed time from Timers
            (Variables|TimeControls|SetTimeSinceLastDodgeStart (Math|Float|Max(Float) (- (Fortnite|Sync|Time|GetServerWorldTimeSeconds) (Variables|TimeControls|GetTimeDodgeStart))))
            (Variables|TimeControls|SetTimeSinceLastCrashStart (Math|Float|Max(Float) (- (Fortnite|Sync|Time|GetServerWorldTimeSeconds) (Variables|TimeControls|GetTimeCrashStart))))
            (Variables|TimeControls|SetTimeSinceLastBoostStart (Math|Float|Max(Float) (- (Fortnite|Sync|Time|GetServerWorldTimeSeconds) (Variables|TimeControls|GetTime_BoostStart))))
            ;[1 0 0] TODO: This is failing to initialize properly in Blueprint Modifier Enabled event
            (Variables|RuntimeVars|SetBaselineFlyingSpeedWithHeightDecay (Math|Float|Sqrt (+ (Math|Float|Square (ScalableFloat|GetValueatLevel (Variables|Balance|Speed|GetRow_FlyingSpeed_Base))) (Math|Float|Square (ScalableFloat|GetValueatLevel (Variables|Balance|GetRow_Flying_HeightDecay_Min))))))
            ;[1 1 1] FOV offsets
            (Variables|RuntimeVars|FOV|SetClampedPreviousFOVOffset (Math|Float|Max(Float) (Variables|FinalOffsets|GetFOVOffset)))
            (if (Class|FortWingedFlyingBlueprintLibrary|FlyingStatusHasAnyFlags (Variables|/Script/SunRoseFlyingRuntime.FortPawnComponent_WingedFlying|GetFlyingStatus (Variables|Default|GetFlyingPawnComponent)) 1)
              (Variables|RuntimeVars|FOV|SetFOVOffset_Input (Math|Interpolation|FInterpTo (Variables|RuntimeVars|FOV|GetFOVOffset_Input) 0.0 (Variables|GetDeltaTime) (Variables|TuningControls|FOV|Interp|GetFOV_ModeSwapInterpRate)))
              (Variables|RuntimeVars|FOV|SetFOVOffset_Speed _returnvalue_7)
              (Variables|RuntimeVars|FOV|SetFOVOffset_Mode (Math|Interpolation|FInterpTo (Variables|RuntimeVars|FOV|GetFOVOffset_Mode) (Math|Float|SelectFloat (Utilities|RemapFloat (Variables|TimeControls|GetTimeSinceLastBoostStart) "(X=0.500000,Y=2.500000)" "(X=20.000000,Y=15.000000)" true) (Utilities|RemapFloat (Variables|TimeControls|GetTimeSinceLastBoostStart) "(X=0.000000,Y=0.500000)" "(X=0.000000,Y=5.000000)" true) (>= (Variables|TimeControls|GetTimeSinceLastBoostStart) 0.5)) (Variables|GetDeltaTime) (Variables|TuningControls|FOV|Interp|GetFOV_ModeSwapInterpRate)))
              (Variables|FinalOffsets|SetFOVOffset (Math|Float|Min(Float) (+ (Variables|RuntimeVars|FOV|GetFOVOffset_Mode) (Math|Float|Max(Float) (Variables|RuntimeVars|FOV|GetFOVOffset_Speed) (Variables|RuntimeVars|FOV|GetFOVOffset_Input))) (Variables|TuningControls|FOV|GetMaxFOVOffset)))
              (elif (Class|FortWingedFlyingBlueprintLibrary|FlyingStatusHasAnyFlags (Variables|/Script/SunRoseFlyingRuntime.FortPawnComponent_WingedFlying|GetFlyingStatus (Variables|Default|GetFlyingPawnComponent)) 2)
                (Variables|RuntimeVars|FOV|SetFOVOffset_Input (Math|Interpolation|FInterpTo (Variables|RuntimeVars|FOV|GetFOVOffset_Input) 0.0 (Variables|GetDeltaTime) (Variables|TuningControls|FOV|Interp|GetFOV_ModeSwapInterpRate)))
                (Variables|RuntimeVars|FOV|SetFOVOffset_Speed (Math|Interpolation|FInterpTo (Variables|RuntimeVars|FOV|GetFOVOffset_Speed) 0.0 (Variables|GetDeltaTime) (Variables|TuningControls|FOV|Interp|GetFOV_ModeSwapInterpRate)))
                (Variables|RuntimeVars|FOV|SetFOVOffset_Mode (Math|Interpolation|FInterpTo (Variables|RuntimeVars|FOV|GetFOVOffset_Mode) 12.0 (Variables|GetDeltaTime) (Variables|TuningControls|FOV|Interp|GetFOV_ModeSwapInterpRate)))
                (Variables|FinalOffsets|SetFOVOffset (Math|Float|Min(Float) (+ (Variables|RuntimeVars|FOV|GetFOVOffset_Mode) (Math|Float|Max(Float) (Variables|RuntimeVars|FOV|GetFOVOffset_Speed) (Variables|RuntimeVars|FOV|GetFOVOffset_Input))) (Variables|TuningControls|FOV|GetMaxFOVOffset)))
                (elif (Class|FortWingedFlyingBlueprintLibrary|FlyingStatusHasAnyFlags (Variables|/Script/SunRoseFlyingRuntime.FortPawnComponent_WingedFlying|GetFlyingStatus (Variables|Default|GetFlyingPawnComponent)) 4)
                  (Variables|RuntimeVars|FOV|SetFOVOffset_Input (Math|Interpolation|FInterpTo (Variables|RuntimeVars|FOV|GetFOVOffset_Input) 0.0 (Variables|GetDeltaTime) (Variables|TuningControls|FOV|Interp|GetFOV_ModeSwapInterpRate)))
                  (Variables|RuntimeVars|FOV|SetFOVOffset_Speed (Math|Interpolation|FInterpTo (Variables|RuntimeVars|FOV|GetFOVOffset_Speed) 0.0 (Variables|GetDeltaTime) (Variables|TuningControls|FOV|Interp|GetFOV_ModeSwapInterpRate)))
                  (Variables|RuntimeVars|FOV|SetFOVOffset_Mode (Math|Interpolation|FInterpTo (Variables|RuntimeVars|FOV|GetFOVOffset_Mode) (Math|Float|SelectFloat (Variables|TuningControls|FOV|GetMaxFOVOffset) -2e+01 (>= (Variables|TimeControls|GetTimeSinceLastCrashStart) (ScalableFloat|GetValueatLevel (Variables|Balance|GetRow_PreCrashWindupDuration)))) (Variables|GetDeltaTime) 3.0))
                  (Variables|FinalOffsets|SetFOVOffset (Math|Float|Min(Float) (+ (Variables|RuntimeVars|FOV|GetFOVOffset_Mode) (Math|Float|Max(Float) (Variables|RuntimeVars|FOV|GetFOVOffset_Speed) (Variables|RuntimeVars|FOV|GetFOVOffset_Input))) (Variables|TuningControls|FOV|GetMaxFOVOffset)))
                  (else
                    (Variables|RuntimeVars|FOV|SetFOVOffset_Input (Math|Interpolation|FInterpTo (Variables|RuntimeVars|FOV|GetFOVOffset_Input) (Math|Float|SelectFloat (Utilities|RemapFloat (Math|Vector2D|BreakVector2D (Class|FortPawnComponentWingedFlying|GetLocalFlyingMoveActionValue (Variables|Default|GetFlyingPawnComponent))) (0.0 1.0) "(X=0.000000,Y=15.000000)" true) (Utilities|RemapFloat (Math|Vector2D|BreakVector2D (Class|FortPawnComponentWingedFlying|GetLocalFlyingMoveActionValue (Variables|Default|GetFlyingPawnComponent))) (0.0 -1.0) "(X=0.000000,Y=-30.000000)" true) (>= (Math|Vector2D|BreakVector2D (Class|FortPawnComponentWingedFlying|GetLocalFlyingMoveActionValue (Variables|Default|GetFlyingPawnComponent))) 0)) (Variables|GetDeltaTime) (Math|Float|SelectFloat 1.4 1.2 (>= (Math|Vector2D|BreakVector2D (Class|FortPawnComponentWingedFlying|GetLocalFlyingMoveActionValue (Variables|Default|GetFlyingPawnComponent))) 0))) :!pinned true) ; Default Flying
                    (Variables|RuntimeVars|FOV|SetFOVOffset_Speed _returnvalue_7)
                    (Variables|RuntimeVars|FOV|SetFOVOffset_Mode (Math|Interpolation|FInterpTo (Variables|RuntimeVars|FOV|GetFOVOffset_Mode) 0.0 (Variables|GetDeltaTime) 1.2))
                    (Variables|FinalOffsets|SetFOVOffset (Math|Float|Min(Float) (+ (Variables|RuntimeVars|FOV|GetFOVOffset_Mode) (Math|Float|Max(Float) (Variables|RuntimeVars|FOV|GetFOVOffset_Speed) (Variables|RuntimeVars|FOV|GetFOVOffset_Input))) (Variables|TuningControls|FOV|GetMaxFOVOffset)))))))
            ;[1 1 1] Forward Location Offsets
            (Variables|RuntimeVars|Offsets|Forward|SetForwardLocationOffsetCorrectionFromFOV (Math|Float|SelectFloat (Utilities|RemapFloat (Variables|FinalOffsets|GetFOVOffset) "(X=0.000000,Y=30.000000)" "(X=0.000000,Y=70.000000)" true) (Utilities|RemapFloat (Variables|FinalOffsets|GetFOVOffset) "(X=0.000000,Y=-30.000000)" "(X=0.000000,Y=-90.000000)" true) (>= (Variables|FinalOffsets|GetFOVOffset) 0)) :!pinned true) ; Forward Location Offset Correction from FOV
            (Variables|RuntimeVars|FOV|SetInterpolatedClampedDeltaFOV (Math|Interpolation|FInterpTo (Variables|RuntimeVars|FOV|GetInterpolatedClampedDeltaFOV) (Math|Float|Max(Float) (/ (- (Variables|FinalOffsets|GetFOVOffset) (Variables|RuntimeVars|FOV|GetClampedPreviousFOVOffset)) (Variables|GetDeltaTime))) (Variables|GetDeltaTime) 3.0))
            ;[1 0 0] TODO: probably need better handling of this to smooth transitions...
            (if (Class|FortWingedFlyingBlueprintLibrary|FlyingStatusHasAnyFlags (Variables|/Script/SunRoseFlyingRuntime.FortPawnComponent_WingedFlying|GetFlyingStatus (Variables|Default|GetFlyingPawnComponent)) 2)
              (Variables|RuntimeVars|Offsets|Forward|SetForwardLocationOffsetFromDeltaFOV (Math|Interpolation|FInterpTo (Variables|RuntimeVars|Offsets|Forward|GetForwardLocationOffsetFromDeltaFOV) 0.0 (Variables|GetDeltaTime) (Variables|TuningControls|FOV|Interp|GetFOV_ModeSwapInterpRate)))
              (Variables|RuntimeVars|Offsets|Forward|SetForwardLocationOffsetFromDownZVelocity (Math|Interpolation|FInterpTo (Variables|RuntimeVars|Offsets|Forward|GetForwardLocationOffsetFromDownZVelocity) 0.0 (Variables|GetDeltaTime) (Variables|TuningControls|FOV|Interp|GetFOV_ModeSwapInterpRate)))
              (elif (Class|FortWingedFlyingBlueprintLibrary|FlyingStatusHasAnyFlags (Variables|/Script/SunRoseFlyingRuntime.FortPawnComponent_WingedFlying|GetFlyingStatus (Variables|Default|GetFlyingPawnComponent)) 4)
                (Variables|RuntimeVars|Offsets|Forward|SetForwardLocationOffsetFromDeltaFOV (Math|Interpolation|FInterpTo (Variables|RuntimeVars|Offsets|Forward|GetForwardLocationOffsetFromDeltaFOV) 0.0 (Variables|GetDeltaTime) (Variables|TuningControls|FOV|Interp|GetFOV_ModeSwapInterpRate)))
                (Variables|RuntimeVars|Offsets|Forward|SetForwardLocationOffsetFromDownZVelocity (Math|Float|SelectFloat (Math|Interpolation|FInterpTo (Variables|RuntimeVars|Offsets|Forward|GetForwardLocationOffsetFromDownZVelocity) (Math|Float|SelectFloat (Variables|TuningControls|Offset|Forward|GetForwardOffset_Distance_Crashing_Final) (Variables|TuningControls|Offset|Forward|GetForwardOffset_Distance_ForceCrashing_Phase1) (>= (Variables|TimeControls|GetTimeSinceLastCrashStart) (ScalableFloat|GetValueatLevel (Variables|Balance|GetRow_PreCrashWindupDuration)))) (Variables|GetDeltaTime) (Variables|TuningControls|Offset|Forward|GetForwardOffset_InterpSpeed_Crashing_Final)) (Math|Interpolation|FInterpTo (Variables|RuntimeVars|Offsets|Forward|GetForwardLocationOffsetFromDownZVelocity) (Math|Float|SelectFloat (Math|Float|SelectFloat (Variables|TuningControls|Offset|Forward|GetForwardOffset_Distance_Crashing_Final) (Variables|TuningControls|Offset|Forward|GetForwardOffset_Distance_Crashing_Phase2) (>= (Variables|TimeControls|GetTimeSinceLastCrashStart) (+ (ScalableFloat|GetValueatLevel (Variables|Balance|GetRow_PreCrashWindupDuration)) (Variables|TuningControls|Offset|Forward|GetForwardOffset_Crashing_TimeAfterWindupToEnterFinalPhase)))) (Variables|TuningControls|Offset|Forward|GetForwardOffset_Distance_Crashing_Phase1) (>= (Variables|TimeControls|GetTimeSinceLastCrashStart) (ScalableFloat|GetValueatLevel (Variables|Balance|GetRow_PreCrashWindupDuration)))) (Variables|GetDeltaTime) (Math|Float|SelectFloat (Math|Float|SelectFloat (Variables|TuningControls|Offset|Forward|GetForwardOffset_InterpSpeed_Crashing_Final) (Variables|TuningControls|Offset|Forward|GetForwardOffset_InterpSpeed_Crashing_Phase2) (>= (Variables|TimeControls|GetTimeSinceLastCrashStart) (+ (ScalableFloat|GetValueatLevel (Variables|Balance|GetRow_PreCrashWindupDuration)) (Variables|TuningControls|Offset|Forward|GetForwardOffset_Crashing_TimeAfterWindupToEnterFinalPhase)))) (Variables|TuningControls|Offset|Forward|GetForwardOffset_InterpSpeed_Crashing_Phase1) (>= (Variables|TimeControls|GetTimeSinceLastCrashStart) (ScalableFloat|GetValueatLevel (Variables|Balance|GetRow_PreCrashWindupDuration))))) (GameplayTags|HasMatchingGameplayTag (Variables|RuntimeVars|GetPlayerPawn) (Variables|Tags|GetTag_ForceCrash))))
                (else
                  (Variables|RuntimeVars|Offsets|Forward|SetForwardLocationOffsetFromDeltaFOV (Math|Interpolation|FInterpTo (Variables|RuntimeVars|Offsets|Forward|GetForwardLocationOffsetFromDeltaFOV) (Utilities|RemapFloat (Math|Float|Power (Variables|RuntimeVars|FOV|GetInterpolatedClampedDeltaFOV) (Variables|RuntimeVars|FOV|GetDeltaFOVForwardOffsetExponent)) (Math|Vector2D|MakeVector2D 0.0 (Math|Float|Power (Variables|TuningControls|FOV|GetMaxFOVOffset) (Variables|RuntimeVars|FOV|GetDeltaFOVForwardOffsetExponent))) "(X=0.000000,Y=-300.000000)" true) (Variables|GetDeltaTime) (Math|Float|SelectFloat 3.0 1.5 (or (<= (Variables|TimeControls|GetTimeSinceLastBoostStart) 1.0) (>= (Variables|TimeControls|GetTimeSinceLastBoostStart) 2.5)))) :!pinned true) ; Default Flying
                  (Variables|RuntimeVars|Offsets|Forward|SetForwardLocationOffsetFromDownZVelocity (Math|Interpolation|FInterpTo (Math|Float|Max(Float) (Variables|RuntimeVars|Offsets|Forward|GetForwardLocationOffsetFromDownZVelocity)) (Utilities|RemapFloat (Variables|RuntimeVars|GetAddedDownZVelocity) (Math|Vector2D|MakeVector2D 0.0 (ScalableFloat|GetValueatLevel (Variables|Balance|Speed|GetRow_FlyingDownZSpeedAddedWhileFalling))) "(X=0.000000,Y=-300.000000)" true) (Variables|GetDeltaTime) 1.3)))))
            (Variables|FinalOffsets|SetViewLocationOffset_Forward (Utilities|Operators|Add (Variables|RuntimeVars|Offsets|Forward|GetForwardLocationOffsetCorrectionFromFOV) (Variables|RuntimeVars|Offsets|Forward|GetForwardLocationOffsetFromDeltaFOV) (Variables|RuntimeVars|Offsets|Forward|GetForwardLocationOffsetFromDownZVelocity))))
          (:then_4
            ;[1 1 1] Right Location Offsets
            (Variables|FinalOffsets|SetViewLocationOffset_Right (Math|Interpolation|FInterpTo (Variables|FinalOffsets|GetViewLocationOffset_Right) (Math|Float|SelectFloat 0.0 (Math|Float|SelectFloat (Math|Float|SelectFloat (Variables|TuningControls|Offset|Right|GetRightOffset_Dodge) (Math|Float|SelectFloat (* (Variables|TuningControls|Offset|Right|GetRightOffset_Dodge) -1.0) 0.0 (Class|FortWingedFlyingBlueprintLibrary|FlyingStatusHasAnyFlags (Variables|/Script/SunRoseFlyingRuntime.FortPawnComponent_WingedFlying|GetFlyingStatus (Variables|Default|GetFlyingPawnComponent)) 16)) (Class|FortWingedFlyingBlueprintLibrary|FlyingStatusHasAnyFlags (Variables|/Script/SunRoseFlyingRuntime.FortPawnComponent_WingedFlying|GetFlyingStatus (Variables|Default|GetFlyingPawnComponent)) 8)) (* (Math|Float|Clamp(Float) (+ (Utilities|RemapFloat (Math|Vector|DotProduct (Variables|/Script/Engine.MovementComponent|GetVelocity (Variables|/Script/Engine.Character|GetCharacterMovement (Variables|RuntimeVars|GetPlayerPawn))) (Transformation|GetActorRightVector (Variables|RuntimeVars|GetPlayerPawn))) (Math|Vector2D|MakeVector2D (* -1.0 (* (/ (ScalableFloat|GetValueatLevel (Variables|Balance|GetRow_Flying_InputWeight_Horizontal)) (Math|Float|Sqrt (+ (Math|Float|Square (ScalableFloat|GetValueatLevel (Variables|Balance|GetRow_Flying_InputWeight_Horizontal))) (Math|Float|Square (ScalableFloat|GetValueatLevel (Variables|Balance|GetRow_Flying_InputWeight_LookDirection)))))) (ScalableFloat|GetValueatLevel (Variables|Balance|Speed|GetRow_FlyingSpeed_Base)))) (* (/ (ScalableFloat|GetValueatLevel (Variables|Balance|GetRow_Flying_InputWeight_Horizontal)) (Math|Float|Sqrt (+ (Math|Float|Square (ScalableFloat|GetValueatLevel (Variables|Balance|GetRow_Flying_InputWeight_Horizontal))) (Math|Float|Square (ScalableFloat|GetValueatLevel (Variables|Balance|GetRow_Flying_InputWeight_LookDirection)))))) (ScalableFloat|GetValueatLevel (Variables|Balance|Speed|GetRow_FlyingSpeed_Base)))) (Math|Vector2D|MakeVector2D (* -1.0 (Variables|TuningControls|Offset|Right|GetRightOffset_Strafe)) (Variables|TuningControls|Offset|Right|GetRightOffset_Strafe)) true) (Utilities|RemapFloat (/ (Math|Rotator|NormalizeAxis (- _returnvalue_yaw_3 (Variables|RuntimeVars|GetPreviousRotation))) (Variables|GetDeltaTime)) (Math|Vector2D|MakeVector2D (* (Variables|TuningControls|Offset|Right|GetRotationRateForMaxRightOffset) -1.0) (Variables|TuningControls|Offset|Right|GetRotationRateForMaxRightOffset)) (Math|Vector2D|MakeVector2D (* -1.0 (Variables|TuningControls|Offset|Right|GetRightOffset_Turn)) (Variables|TuningControls|Offset|Right|GetRightOffset_Turn)) true)) (* -1.0 (Variables|TuningControls|Offset|Right|GetRightOffset_MaxCombinedStrafeTurn)) (Variables|TuningControls|Offset|Right|GetRightOffset_MaxCombinedStrafeTurn)) (Utilities|RemapFloat (Variables|FinalOffsets|GetFOVOffset) (0.0 15.0) "(X=1.000000,Y=0.670000)" true)) (and (Variables|TimeControls|GetHasSetDodgeStartTime) (<= (Variables|TimeControls|GetTimeSinceLastDodgeStart) 0.35))) (Class|FortWingedFlyingBlueprintLibrary|FlyingStatusHasAnyFlags (Variables|/Script/SunRoseFlyingRuntime.FortPawnComponent_WingedFlying|GetFlyingStatus (Variables|Default|GetFlyingPawnComponent)) 4)) (Variables|GetDeltaTime) (Math|Float|SelectFloat 4.25 (Math|Float|SelectFloat 5.0 (Math|Float|SelectFloat 0.8 2.2 (> (Math|Float|Absolute(Float) (Utilities|RemapFloat (Math|Vector|DotProduct (Variables|/Script/Engine.MovementComponent|GetVelocity (Variables|/Script/Engine.Character|GetCharacterMovement (Variables|RuntimeVars|GetPlayerPawn))) (Transformation|GetActorRightVector (Variables|RuntimeVars|GetPlayerPawn))) (Math|Vector2D|MakeVector2D (* -1.0 (* (/ (ScalableFloat|GetValueatLevel (Variables|Balance|GetRow_Flying_InputWeight_Horizontal)) (Math|Float|Sqrt (+ (Math|Float|Square (ScalableFloat|GetValueatLevel (Variables|Balance|GetRow_Flying_InputWeight_Horizontal))) (Math|Float|Square (ScalableFloat|GetValueatLevel (Variables|Balance|GetRow_Flying_InputWeight_LookDirection)))))) (ScalableFloat|GetValueatLevel (Variables|Balance|Speed|GetRow_FlyingSpeed_Base)))) (* (/ (ScalableFloat|GetValueatLevel (Variables|Balance|GetRow_Flying_InputWeight_Horizontal)) (Math|Float|Sqrt (+ (Math|Float|Square (ScalableFloat|GetValueatLevel (Variables|Balance|GetRow_Flying_InputWeight_Horizontal))) (Math|Float|Square (ScalableFloat|GetValueatLevel (Variables|Balance|GetRow_Flying_InputWeight_LookDirection)))))) (ScalableFloat|GetValueatLevel (Variables|Balance|Speed|GetRow_FlyingSpeed_Base)))) (Math|Vector2D|MakeVector2D (* -1.0 (Variables|TuningControls|Offset|Right|GetRightOffset_Strafe)) (Variables|TuningControls|Offset|Right|GetRightOffset_Strafe)) true)) (Math|Float|Absolute(Float) (Utilities|RemapFloat (/ (Math|Rotator|NormalizeAxis (- _returnvalue_yaw_3 (Variables|RuntimeVars|GetPreviousRotation))) (Variables|GetDeltaTime)) (Math|Vector2D|MakeVector2D (* (Variables|TuningControls|Offset|Right|GetRotationRateForMaxRightOffset) -1.0) (Variables|TuningControls|Offset|Right|GetRotationRateForMaxRightOffset)) (Math|Vector2D|MakeVector2D (* -1.0 (Variables|TuningControls|Offset|Right|GetRightOffset_Turn)) (Variables|TuningControls|Offset|Right|GetRightOffset_Turn)) true)))) (and (> (Variables|TimeControls|GetTimeSinceLastDodgeStart) 0) (<= (Variables|TimeControls|GetTimeSinceLastDodgeStart) 1.0))) (Class|FortWingedFlyingBlueprintLibrary|FlyingStatusHasAnyFlags (Variables|/Script/SunRoseFlyingRuntime.FortPawnComponent_WingedFlying|GetFlyingStatus (Variables|Default|GetFlyingPawnComponent)) 4))))
            (Variables|RuntimeVars|SetPreviousRotation _returnvalue_yaw_2))
          (:then_5
            ;[1 0.7819 0] Debug print
            ;[1 1 1] Return
            (return (+ (Variables|GetViewLocation) (* (Utilities|Operators|Add (* (Math|Vector|GetForwardVector (Variables|GetViewRotation)) (Variables|FinalOffsets|GetViewLocationOffset_Forward)) (* (Math|Vector|GetRightVector (Variables|GetViewRotation)) (Variables|FinalOffsets|GetViewLocationOffset_Right)) (* (Math|Vector|GetUpVector (Variables|GetViewRotation)) (Variables|FinalOffsets|GetViewLocationOffset_Up))) (Variables|CameraModifier|GetAlpha))) (Variables|GetViewRotation) (+ (Variables|GetFOV) (* (Variables|CameraModifier|GetAlpha) (Variables|FinalOffsets|GetFOVOffset)))))))
      (else
        (return ViewLocation ViewRotation FOV)))))
